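# MetaSound Source preset that plays Henchman 1's post-teleport voice line in VividRazor Minions, as a mono one-shot.
# It inherits its graph from the shared MSS_VO_BR_Boss parent and supplies two FullAlert waves, with delay and pitch/volume variation zeroed and probability at full.
# It uses the shared alerted NPC attenuation, critical enemy VO sound class and boss combat chatter concurrency, alongside a one-per-owner, prevent-new override setting.
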MetaSoundSource /VividRazor_Minions/Audio/VO/Henchman_1/MSP_VO_S34_Henchman_1_PostTeleport
preset_parent=/Game/Sounds/Fort_Human_VO_BR/NPCs/Shared/MSS_VO_BR_Boss
inputs [
 "Delay Max (Seconds)" {type=Time}
 "Delay Min (Seconds)" {type=Time}
 "Probability (0-1)" {type=Float}
 "Random Pitch Max (Semitones)" {type=Float}
 "Random Pitch Min (Semitones)" {type=Float}
 "Random Volume Change Max (Decibels)" {type=Float}
 "Random Volume Change Min (Decibels)" {type=Float}
 UE.Source.OnPlay {type=Trigger}
 Wave Array {type=WaveAsset:Array}
]
interfaces [UE.OutputFormat.Mono:1.0, UE.Source.OneShot:1.0, UE.Source:1.1]
outputs [
 "UE.OutputFormat.Mono.Audio:0" {type=Audio}
 UE.Source.OneShot.OnFinished {type=Trigger}
]
pages [Default {input_defaults {
   "Delay Max (Seconds)"=0
   "Delay Min (Seconds)"=0
   "Probability (0-1)"=1
   "Random Pitch Max (Semitones)"=0
   "Random Pitch Min (Semitones)"=0
   "Random Volume Change Max (Decibels)"=0
   "Random Volume Change Min (Decibels)"=0
   UE.Source.OnPlay=null
   Wave Array [
    /VividRazor_Minions/Audio/VO/Henchman_1/Waves/VO_S34_Henchman_1_FullAlert_010
    /VividRazor_Minions/Audio/VO/Henchman_1/Waves/VO_S34_Henchman_1_FullAlert_011
   ]
  }}]
properties {
 assetImportData=&self:AssetImportData
 attenuationSettings=/Game/Sounds/Fort_Human_VO_BR/NPCs/Shared/NPC_VO_Alerted_Att
 maxDistance=12500
 sampleRate=48000
 soundClassObject=&Fort_VO_Enemy_Critical
 volume=2
 concurrencyOverrides {bLimitToOwner=true, maxCount=1, resolutionRule=PreventNew}
 concurrencySet [&NPC_Boss_Combat_Chatter_Conc]
}
